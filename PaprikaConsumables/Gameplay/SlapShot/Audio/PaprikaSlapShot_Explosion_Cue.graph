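# Sound cue for the Paprika Slap Shot consumable's explosion.
# A Random node picks among four explosion wave variants with equal weight.
# It plays at reduced volume under the medium explosives sound class, spatialized with a borrowed smoke grenade interior attenuation asset.
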SoundCue /PaprikaConsumables/Gameplay/SlapShot/Audio/PaprikaSlapShot_Explosion_Cue
root=Random_0
nodes [
 {id=Random_0, children [WavePlayer_0,WavePlayer_3,WavePlayer_4,WavePlayer_7], properties {weights [1,1,1,1]}}
 {id=WavePlayer_0, properties {soundWaveAssetPtr=&/PaprikaSlapShot_Explosion_01}}
 {id=WavePlayer_3, properties {soundWaveAssetPtr=&/PaprikaSlapShot_Explosion_04}}
 {id=WavePlayer_4, properties {soundWaveAssetPtr=&/PaprikaSlapShot_Explosion_05}}
 {id=WavePlayer_7, properties {soundWaveAssetPtr=&/PaprikaSlapShot_Explosion_08}}
]
properties {
 attenuationSettings=/Game/Sounds/Attenuation/Fort_Grenade_Smoke_Inside_Att
 maxDistance=2500
 soundClassObject=/Game/Sounds/FortSoundClassesAndModes/Fort_Explosives_Medium_PSM
 volumeMultiplier=0.6
}
