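# Sound cue for interacting with the SlapShot consumable, routed through the loot collect sound class.
# A random node picks one of three equally weighted interact waves, then a modulator varies its volume and raises pitch slightly.
# It borrows prop search attenuation and overrides concurrency to cap simultaneous plays, stopping the farthest instance first.
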SoundCue /PaprikaConsumables/Gameplay/SlapShot/Audio/PaprikaSlapShot_Interact_Cue
root=Modulator_15
nodes [
 {id=Modulator_15, children [Random_0], properties {pitchMax=1.07, pitchMin=1, volumeMax=1.07, volumeMin=0.92}}
 {id=Random_0, children [WavePlayer_0,WavePlayer_1,WavePlayer_2], properties {weights [1,1,1]}}
 {id=WavePlayer_0, properties {soundWaveAssetPtr=&/PaprikaSlapShot_Interact_01}}
 {id=WavePlayer_1, properties {soundWaveAssetPtr=&/PaprikaSlapShot_Interact_02}}
 {id=WavePlayer_2, properties {soundWaveAssetPtr=&/PaprikaSlapShot_Interact_03}}
]
properties {
 attenuationSettings=/Game/Sounds/Attenuation/Fort_Prop_Search_Att
 bOverrideConcurrency=true
 maxDistance=1300
 soundClassObject=&Fort_Loot_Collect
 volumeMultiplier=1
 concurrencyOverrides {maxCount=4, resolutionRule=StopFarthestThenPreventNew}
}
