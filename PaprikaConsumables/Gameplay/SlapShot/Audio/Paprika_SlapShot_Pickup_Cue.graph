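# Sound cue for the pickup sound of the SlapShot consumable in PaprikaConsumables.
# A Random node picks one of three pickup wave variations with equal weight.
# It plays through the shared Fort weapon pickup attenuation and loot collect sound class, with boosted volume.
# A concurrency override limits it to one playing instance per owner.
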SoundCue /PaprikaConsumables/Gameplay/SlapShot/Audio/Paprika_SlapShot_Pickup_Cue
root=Random_0
nodes [
 {id=Random_0, children [WavePlayer_0,WavePlayer_4,WavePlayer_5], properties {weights [1,1,1]}}
 {id=WavePlayer_0, properties {soundWaveAssetPtr=&/Paprika_SlapShot_Pickup_01}}
 {id=WavePlayer_4, properties {soundWaveAssetPtr=&/Paprika_SlapShot_Pickup_02}}
 {id=WavePlayer_5, properties {soundWaveAssetPtr=&/Paprika_SlapShot_Pickup_03}}
]
properties {
 attenuationSettings=&Fort_Weapon_Pickup_Att
 bOverrideConcurrency=true
 maxDistance=3500
 soundClassObject=&Fort_Loot_Collect
 volumeMultiplier=1.7
 concurrencyOverrides {bLimitToOwner=true, maxCount=1}
}
