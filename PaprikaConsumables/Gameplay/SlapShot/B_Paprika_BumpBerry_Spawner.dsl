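; Blueprint spawner actor for the BumpBerry consumable (Paprika SlapShot) that drops a pickup into the world.
; A spawn event calls SpawnPickupInWorld for one of the configured item ID at the actor's location and forward vector, tagged as floor loot and loot drop.
; When the destroy-after-spawn flag is set, it then calls KillSpawner.
; The construction script only runs its parent.
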(event EventSpawnItem
  (Inventory|SpawnPickupInWorld (Variables|Default|GetIDToSpawn) 1 (Transformation|GetActorLocation) (Math|Rotator|GetRotationXVector (Transformation|GetActorRotation)) 0 false false false -1 "FloorLoot" "LootDrop")
  (if (Variables|Default|GetDestroyAfterSpawn)
    (CallFunction|KillSpawner)))

(fn ConstructionScript
  (|CallParentFunction))
